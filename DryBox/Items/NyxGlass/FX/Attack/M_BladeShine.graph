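# M_BladeShine is an unlit additive material for Niagara mesh particles, drawing a blue shine for the NyxGlass attack effect.
# The Sheen dynamic parameter offsets U, which mixes with V through Band and C to form a slanted band through MF_FX_TilingMethod.
# The tinted band adds to a second colored glow, shaped by a sine falloff across V and a fade along U, then passes through MF_FX_EmissiveAdjustments.
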Material /DryBox/Items/NyxGlass/FX/Attack/M_BladeShine
graph {
 properties {
  bUsedWithNiagaraMeshParticles=true
  blendMode=BLEND_Additive
  pixelDepthOffsetMode=PDOM_Legacy
  shadingModel=MSM_Unlit
  usedShadingModels=Unlit
 }
 nodes [
  n1 = Add(
   A=MaterialFunctionCall(Float2=TextureCoordinate()) {materialFunction=&BreakOutFloat2Components}.R
   B=DynamicParameter() {defaultValue=<rgba 0 1 1 1>, paramNames [Sheen,Param2,Param3,Param4]}.Sheen
  )
  n2 = ScalarParameter() {defaultValue=4, parameterName=Band}
  n3 = MaterialFunctionCall(Float2=TextureCoordinate()) {materialFunction=&BreakOutFloat2Components}
  n4 = ScalarParameter() {defaultValue=-0.15, parameterName=C}
  n5 = MaterialFunctionCall!(TexCoord=Add(A=Multiply(A=@n1, B=@n2) {constB=2}, B=Multiply(A=@n3.G, B=@n4)), ClampChannel=Constant2Vector() {g=1, r=1}) {materialFunction=/Game/Effects/Fort_Effects/MaterialFunctions/MF_FX_TilingMethod}
  Add(A=Multiply(A=ParticleColor().A), B=@n5.Result)
  n6 = Multiply(A=@n5.Result, B=VectorParameter() {parameterName=Param_1, defaultValue=<rgba 0 1.1667 7 1>}.RGB)
  n7 = MaterialFunctionCall(Float2=TextureCoordinate()) {materialFunction=&BreakOutFloat2Components}
  n8 = OneMinus(None=@n7.R)
  n9 = Power(Base=Sine(None=Multiply(A=@n7.G) {constB=0.5})) {constExponent=5}
  n10 = Multiply(
   A=Multiply(A=VectorParameter() {defaultValue=<rgba 0 0.5 3 1>}.RGB, B=Power(Base=@n8) {constExponent=0.5})
   B=Add(A=Multiply(A=@n9)) {constB=0}
  )
  MaterialFunctionCall(Emissive=Divide(A=Add(A=@n6, B=@n10)) {constB=1.85}) {materialFunction=&MF_FX_EmissiveAdjustments}
  Sine(None=Multiply(A=@n7.R) {constB=0.5})
 ]
}
